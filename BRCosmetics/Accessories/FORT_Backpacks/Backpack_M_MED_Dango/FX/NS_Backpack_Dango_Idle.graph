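# Idle Niagara effect for the Dango backpack, producing drips and drool at skeletal mesh sockets, with bursts suppressed while skydiving.
# Drip_A sprites cling, swell and send death-event positions; Drip_B sprites fall and shrink, with no spawn module of their own.
# Drool meshes hang from two sockets using pendulum constraints, feeding pivot and tip positions into dynamic material parameters.
# All emitters choose color by time of day, darkening at night.
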NiagaraSystem /BRCosmetics/Accessories/FORT_Backpacks/Backpack_M_MED_Dango/FX/NS_Backpack_Dango_Idle
dependencies {
 usedDataInterfaces [
  &NiagaraDataInterfaceArrayInt32
  &NiagaraDataInterfaceVector2DCurve
  &NiagaraDataInterfaceCurve
  &NiagaraDataInterfaceVectorCurve
 ]
 usedDynamicInputs [&FN_FrontEndFloatModifier]
 usedModules [
  &SystemState
  &self:SystemSpawnScript.SystemScriptSource.SystemScriptGraph.NiagaraNodeAssignment_0.SetVariables_#1
  &self:SystemSpawnScript.SystemScriptSource.SystemScriptGraph.NiagaraNodeAssignment_1.SetVariables_#2
  &self:SystemSpawnScript.SystemScriptSource.SystemScriptGraph.NiagaraNodeAssignment_2.SetVariables_#3
 ]
}
emitters {
 Drip_A {
  emitterUpdateScript [
   &EmitterState(
    Life Cycle Mode=::NewEnumerator1
    Loop Duration=&UniformRangedFloat(Minimum=3, Maximum=6)
    Recalculate Duration Each Loop=true
    UseLoopDelay=false
   )
   &SpawnBurst_Instantaneous(
    Spawn Count=&MakeCustomIntFromBool(Bool=$User.IsSkydiving, True INT=0, False INT=1)
    Use Spawn Probability=false
   )
  ]
  particleSpawnScript [&InitializeParticle(
    Color=&FN_Select_Color_from_Time_Of_Day(Morning=<rgba 1 1 1 1>, Day=<rgba 1 1 1 1>, Evening=<rgba 1 1 1 1>, Night=<rgba 0.5 0.5 0.5 1>)
    Lifetime=&RandomRangeFloat(Minimum=1.8, Maximum=2.1)
    Sprite Rotation Angle=180
    Sprite Rotation Mode=::NewEnumerator0
    Sprite Size Mode=::NewEnumerator3
    Uniform Sprite Size=&RandomRangeFloat(Minimum=1, Maximum=1.2)
    UsePositionOffset=false
    Write Material Random=true
   ) {module=InitializeParticle001}]
  particleUpdateScript [
   &ParticleState()
   &SkeletalMeshLocation(
    Bone / Socket Index=&SelectIntFromArray(
     Array Sampling Mode=::NewEnumerator1
     Direct Array Index=?
     Int Selection Array=DataInterface() {propertyValues {IntData [1,2]}}
     Random Seed=0
    )
    Inherit Velocity=false
    MaskSpawnGroup=false
    Mesh Sampling Type=::NewEnumerator5
    Sampled Position Offset=<0 0 3>
    Skeletal Mesh=$User.SkeletalMesh
    Socket Sampling Mode=::NewEnumerator5
   )
   &SolveForcesAndVelocity(Clamp Velocity=false, Limit Acceleration=false)
   &ScaleSpriteSize(Scale Factor=&Vector2DFromCurve(
     Vector2Curve=DataInterface() {propertyValues {
       XCurve {keys [
         {interpMode=RCIM_Cubic, tangentMode=RCTM_User, time=0, value=1.15}
         {arriveTangent=-0.8621, interpMode=RCIM_Cubic, leaveTangent=-0.8621, tangentMode=RCTM_User, time=0.6, value=1.14}
         {arriveTangent=-0.0055, interpMode=RCIM_Cubic, leaveTangent=-0.0055, tangentMode=RCTM_User, time=0.8473, value=0.6219}
         {interpMode=RCIM_Cubic, tangentMode=RCTM_User, time=1.0017, value=0.6786}
        ]}
       YCurve {keys [
         {interpMode=RCIM_Cubic, tangentMode=RCTM_User, time=-0.0019, value=0.3593}
         {arriveTangent=2.8736, interpMode=RCIM_Cubic, leaveTangent=2.8736, tangentMode=RCTM_User, time=0.4, value=0.99}
         {arriveTangent=0.0726, interpMode=RCIM_Cubic, leaveTangent=0.0726, tangentMode=RCTM_User, time=0.7, value=1.8}
         {interpMode=RCIM_Cubic, tangentMode=RCTM_User, time=0.9965, value=1.0355}
        ]}
      }}
     ScaleCurve=<1.2 3>
    ))
   &GenerateDeathEvent(
    Gap Correction Amount=1.5
    Position To Send:Vector3f=/Niagara/DynamicInputs/Transforms/TransformPosition(
     Position=&Add_Vector(A=Linked() {linkedVariable {name=Particles.Position, type=NiagaraPosition}}, B=<0 0 -3.5>)
     Source Space=::Local
     Destination Space=::World
    )
   )
   &ScaleColor(
    Scale Alpha=&FloatFromCurve(FloatCurve=DataInterface() {propertyValues {Curve {keys [
         <at 0 0 RCIM_Cubic>
         {arriveTangent=-0.0188, interpMode=RCIM_Cubic, leaveTangent=-0.0188, tangentMode=RCTM_User, time=0.1, value=1}
         <at 0.9 1>
        ]}}})
    ScaleA=true
    ScaleRGB=true
   )
  ]
  rendererClasses [&NiagaraSpriteRendererProperties]
  renderers [{rendererClass=&NiagaraSpriteRendererProperties}]
 }
 Drip_B {
  emitterUpdateScript [&EmitterState(Life Cycle Mode=::NewEnumerator1, Loop Duration=1, UseLoopDelay=false)]
  particleSpawnScript [
   &InitializeParticle(
    Color=&FN_Select_Color_from_Time_Of_Day(Morning=<rgba 1 1 1 1>, Day=<rgba 1 1 1 1>, Evening=<rgba 1 1 1 1>, Night=<rgba 0.5 0.5 0.5 1>)
    Lifetime=&RandomRangeFloat(Minimum=0.7, Maximum=1)
    Sprite Size Mode=::NewEnumerator3
    Uniform Sprite Size=&RandomRangeFloat(Minimum=2, Maximum=3)
    UsePositionOffset=false
    Write Material Random=true
   ) {module=InitializeParticle001}
   &InheritVelocity(Module.Limit Speed=false, Velocity Scale=<0.5 0.5 0.5>) {enabled=false}
  ]
  particleUpdateScript [
   &ParticleState()
   &AccelerationForce(Acceleration=&Multiply_VectorByFloat(
     Vector=&VectorFromCurve(VectorCurve=DataInterface() {propertyValues {
        XCurve {keys [<at 0 0>,<at 0.9988 0>]}
        YCurve {keys [<at 0 0>,<at 0.9988 0>]}
        ZCurve {keys [
          {arriveTangent=-15.5802, interpMode=RCIM_Cubic, leaveTangent=-8.2205, tangentMode=RCTM_User, time=0, value=0}
          {arriveTangent=-962.5264, interpMode=RCIM_Cubic, leaveTangent=-1066.8696, tangentMode=RCTM_User, time=0.3, value=-250}
         ]}
       }})
     Float=1
    ))
   &SolveForcesAndVelocity(Clamp Velocity=false, Limit Acceleration=false)
   &ScaleSpriteSize(Scale Factor=&Vector2DFromCurve(
     Vector2Curve=DataInterface() {propertyValues {
       XCurve {keys [
         <at 0 0 RCIM_Cubic>
         {arriveTangent=-0.1, interpMode=RCIM_Cubic, leaveTangent=-0.1, tangentMode=RCTM_User, time=0.18, value=0.99}
         {interpMode=RCIM_Cubic, tangentMode=RCTM_User, time=1, value=0}
        ]}
       YCurve {keys [
         <at 0 0 RCIM_Cubic>
         {arriveTangent=-0.1001, interpMode=RCIM_Cubic, leaveTangent=-0.1001, tangentMode=RCTM_User, time=0.18, value=0.99}
         {interpMode=RCIM_Cubic, tangentMode=RCTM_User, time=0.9989, value=0}
        ]}
      }}
     ScaleCurve=<1 1>
    ))
   &ScaleSpriteSize(Scale Factor=<0.5 1.3>, Scale Sprite Size Mode=::NewEnumerator2) {module=ScaleSpriteSize001}
   /SetVariables(Particles.PositionAlt:NiagaraPosition=&Add_Vector(
     A=Linked() {linkedVariable {name=Particles.Position, type=NiagaraPosition}}
     B=&RandomRangeVector(Minimum=<0 -0.1 0>, Maximum=<0 -0.1 0>)
    )) {isSetParameters=true}
  ]
  rendererClasses [&NiagaraSpriteRendererProperties]
  renderers [{rendererClass=&NiagaraSpriteRendererProperties}]
 }
 Drool {
  emitterUpdateScript [
   &EmitterState(
    Life Cycle Mode=::NewEnumerator1
    Loop Duration=&RandomRangeFloat(Minimum=2, Maximum=4)
    UseLoopDelay=false
   )
   &SpawnBurst_Instantaneous(
    Spawn Count=&MakeCustomIntFromBool(Bool=$User.IsSkydiving, True INT=0, False INT=1)
    Use Spawn Probability=false
   )
  ]
  particleSpawnScript [
   /SetVariables(Particles.PendulumDrag=5, Particles.PendulumLengthScale=0.05, Particles.PendulumMaskFalloff=1.5, Particles.PendulumRigidity=0.2) {isSetParameters=true}
   &FN_SetSpawnGroup(Number of Spawn Groups=2)
   /Niagara/Modules/Spawn/Initialization/InitializeParticle(
    Color=&FN_Select_Color_from_Time_Of_Day(
     Morning=<rgba 1 1 1 1>
     Day=<rgba 1 1 1 1>
     Evening=<rgba 1 1 1 1>
     Night=&MaskLinearColorBySpawnGroup(Masked Linear Color=<rgba 0.5 0.5 0.5 1>, Passthrough Linear Color=<rgba 0.4 0.4 0.4 1>)
    )
    Lifetime=&UniformRangedFloat(Minimum=2, Maximum=3.5)
    Mesh Scale=&Multiply_Vector(A=<1 1 1>, B=&MakeVector(X=0.08, Y=0.08, Z=$Particles.PendulumLengthScale))
    Module.Write Color=true
    Module.Write Lifetime=true
    Module.Write Mass=true
    Module.Write Material Random=true
    Module.Write Position=true
    Module.Write Scale=true
    Module.Write SpriteRotation=false
    Module.Write SpriteSize=true
    Sprite Size=<1 1>
   )
  ]
  particleUpdateScript [
   &SkeletalMeshLocation(
    Apply To Particle Position=false
    Bone / Socket Index=&SelectIntFromArray(Direct Array Index=$Particles.SpawnGroup, Int Selection Array=DataInterface() {propertyValues {IntData [0,3]}})
    Inherit Velocity=false
    MaskSpawnGroup=false
    Mesh Sampling Type=::NewEnumerator5
    Skeletal Mesh=$User.SkeletalMesh
    Socket Sampling Mode=::NewEnumerator5
   )
   &ParticleState()
   /SetVariables(Particles.MeshOrigin:Vector3f=&Add_Vector(
     A=$Output.SkeletalMeshLocation.SampledPosition
     B=&MakeVector(
      X=0
      Y=0
      Z=&Add_Float(
       A=&FloatFromCurve(
        FloatCurve=DataInterface() {propertyValues {Curve {keys [
            {arriveTangent=0.368, interpMode=RCIM_Cubic, leaveTangent=0.368, tangentMode=RCTM_User, time=0.7, value=0}
            {arriveTangent=-4.6848, interpMode=RCIM_Cubic, leaveTangent=-4.6848, tangentMode=RCTM_User, time=1, value=-1}
           ]}}}
        Scale Curve=170
       )
       B=2.7
      )
     )
    )) {isSetParameters=true}
   &Drag(Drag=$Particles.PendulumDrag, Use Linear Drag=true, Use Rotational Drag=false)
   &GravityForce(Gravity=&Multiply_VectorByFloat(Vector=<0 0 -700>, Float=1))
   &PendulumSetup(
    Module.Enable Spring Driver=false
    Pendulum Length:NiagaraFloat=&Multiply_Float(A=100, B=$Particles.PendulumLengthScale)
    Pendulum Pivot:Vector3f=$Particles.MeshOrigin
    "Spring Driven Constraint (Tightness)"=3
   )
   &SolveForcesAndVelocity(Clamp Velocity=false, Limit Acceleration=false)
   &PendulumConstraint()
   &CalculateAccurateVelocity()
   /SetVariables(Particles.MeshFacing:Vector3f=&NormalizeVector(Vector To Normalize=&Subtract_Vector(A=$Particles.MeshOrigin, B=Linked() {linkedVariable {name=Particles.Position, type=NiagaraPosition}}))) {isSetParameters=true}
   /Niagara/Modules/Update/Orientation/V2/OrientMeshToVector(
    Look At Direction:Vector3f=&Multiply_VectorByFloat(Vector=$Particles.MeshFacing, Float=1)
    Override Quat=false
    Reference Axis Vector:Vector3f=<0 0 1>
    UseOptionalSideAxis=false
    UseOptionalUpAxis=false
   )
   &DynamicMaterialParameters(
    Index 2 Param 1=HlslExpression() {hlslExpression=Particles.MeshOrigin.x}
    Index 2 Param 2=HlslExpression() {hlslExpression=Particles.MeshOrigin.y}
    Index 2 Param 3=HlslExpression() {hlslExpression=Particles.MeshOrigin.z}
    Index 2 Param 4=$Particles.PendulumMaskFalloff
    Index 3 Param 1=&MakeFloatFromVector(Vector=Linked() {linkedVariable {name=Particles.Position, type=NiagaraPosition}}, Channel=::NewEnumerator0)
    Index 3 Param 2=&MakeFloatFromVector(Vector=Linked() {linkedVariable {name=Particles.Position, type=NiagaraPosition}}, Channel=::NewEnumerator1)
    Index 3 Param 3=&MakeFloatFromVector(Vector=Linked() {linkedVariable {name=Particles.Position, type=NiagaraPosition}}, Channel=::NewEnumerator2)
    Index 3 Param 4=$Particles.PendulumRigidity
    Module.Param0WriteEnabled2=true
    Module.Param0WriteEnabled3=true
    Module.Param1WriteEnabled2=true
    Module.Param1WriteEnabled3=true
    Module.Param2WriteEnabled2=true
    Module.Param2WriteEnabled3=true
    Module.Param3WriteEnabled2=true
    Module.Param3WriteEnabled3=true
    Write Parameter Index 0=false
    Write Parameter Index 2=true
    Write Parameter Index 3=true
   )
   &ScaleMeshSize(Scale Factor=&MakeVector(
     X=&FloatFromCurve(FloatCurve=DataInterface() {propertyValues {Curve {keys [<at 0 0.2>,<at 0.7 1>]}}})
     Y=&FloatFromCurve(FloatCurve=DataInterface() {propertyValues {Curve {keys [<at 0 0.2>,<at 0.7 1>]}}})
     Z=&FloatFromCurve(FloatCurve=DataInterface() {propertyValues {Curve {keys [
          <at 0 0.25 RCIM_Cubic>
          <at 0.7 0.8055 RCIM_Cubic>
          {arriveTangent=-1.5491, interpMode=RCIM_Cubic, leaveTangent=-1.5491, tangentMode=RCTM_User, time=1, value=0.25}
         ]}}})
    ))
   &ScaleColor(
    Scale RGB=&MaskVectorBySpawnGroup(Masked Vector=<1 1 1>, Passthrough Vector=<2 0.65 0.5>)
    ScaleA=true
    ScaleRGB=true
   )
  ]
  rendererClasses [&NiagaraMeshRendererProperties]
  renderers [{rendererClass=&NiagaraMeshRendererProperties, rendererIndex=0}]
 }
}
summary {
 systemName=NS_Backpack_Dango_Idle
 emitters [
  {emitterName=Drip_A, rendererClasses [&NiagaraSpriteRendererProperties]}
  {emitterName=Drip_B, rendererClasses [&NiagaraSpriteRendererProperties]}
  {emitterName=Drool, rendererClasses [&NiagaraMeshRendererProperties]}
 ]
 userVariables [
  User.bIsFrontEnd {type=NiagaraBool}
  User.bIsFrontEndPreview {type=NiagaraBool}
  User.IsSkydiving {type=NiagaraBool}
  User.SkeletalMesh {
   type=NiagaraDataInterfaceSkeletalMesh
   defaultValue {
    struct=NiagaraExt_VariableValue_DataInterface
    value {dataInterface=&self:NiagaraDataInterfaceSkeletalMesh_0}
   }
  }
 ]
}
